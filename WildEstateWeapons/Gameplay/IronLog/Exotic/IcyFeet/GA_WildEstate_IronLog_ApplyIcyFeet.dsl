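; GA_WildEstate_IronLog_ApplyIcyFeet is a gameplay ability that puts the IcyFeet effect on enemies for the exotic IronLog weapon.
; It activates from a gameplay event and reads the instigator and target from the event data.
; If the target's team affiliation is hostile to the instigator, it applies the configured effects with a stored source position.
; The ability then ends immediately.
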(event Ability|EventActivateAbilityFromEvent
  (params (EventData :type Gameplay%20Event%20Data%20Structure%20%28by%20ref%29))
  (bind (_eventtag _instigator _target) (Utilities|Struct|BreakGameplayEventData EventData))
  ;[1 1 1] Apply IcyFeet to target
  (if (Utilities|Enum|Equal(Enum) (Team|GetTeamAffiliation _instigator _target) "Hostile")
    (Fortnite|ApplyGameplayEffectsToActor _target (Variables|Default|GetEffects%20to%20Apply) (Variables|Default|GetSource%20Pos))
    (else
      _))
  (Ability|EndAbility))
